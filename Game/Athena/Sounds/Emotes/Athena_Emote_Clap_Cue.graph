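# Sound cue for the Athena clap emote, playing the dance_soldier_clap wave through a modulator that slightly randomizes volume with fixed pitch.
# It uses the Athena emote attenuation and the emote music sound class, and sends to the Fortnite music audio-analysis submix.
# An Athena emote music concurrency set is assigned, alongside override values for a small per-owner limit that stops the oldest instance.
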SoundCue /Game/Athena/Sounds/Emotes/Athena_Emote_Clap_Cue
root=Modulator_0
nodes [
 {id=Modulator_0, children [WavePlayer_37], properties {pitchMax=1, pitchMin=1, volumeMax=1, volumeMin=0.8}}
 {id=WavePlayer_37, properties {soundWaveAssetPtr=/Game/Sounds/Fort_Music/Dance/dance_soldier_clap}}
]
properties {
 attenuationSettings=&Fort_Emotes_Athena_Att
 maxDistance=1600
 soundClassObject=&Fort_Music_Emote
 volumeMultiplier=1
 concurrencyOverrides {bLimitToOwner=true, maxCount=2, resolutionRule=StopOldest}
 concurrencySet [/Game/Sounds/Concurrency_Rules/Athena_EmoteMusic_Conc]
 soundSubmixSends [{minSendLevel=0, soundSubmix=&Fortnite_AudioAnalysis_Music}]
}
