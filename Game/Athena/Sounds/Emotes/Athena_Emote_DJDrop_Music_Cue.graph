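# Sound cue for the first-person DJ Drop emote music in Athena.
# A quality level node selects between standard and "opt" variants, each mixing a one-shot intro with a looping body that a fixed delay starts 7.5 seconds later.
# It uses the first-person emote music class and concurrency set, and sends to the music audio analysis submix.
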SoundCue /Game/Athena/Sounds/Emotes/Athena_Emote_DJDrop_Music_Cue
root=QualityLevel_0
nodes [
 {id=Delay_0, children [WavePlayer_2], properties {delayMax=7.5, delayMin=7.5}}
 {id=Delay_1, children [WavePlayer_3], properties {delayMax=7.5, delayMin=7.5}}
 {id=Mixer_0, children [WavePlayer_0,Delay_0], properties {inputVolume [1,1]}}
 {id=Mixer_1, children [WavePlayer_1,Delay_1], properties {inputVolume [1,1]}}
 {id=QualityLevel_0, children [Mixer_0,Mixer_0,Mixer_1,Mixer_1]}
 {id=WavePlayer_0, properties {soundWaveAssetPtr=&/Athena_Emotes_DJDrop_Intro}}
 {id=WavePlayer_1, properties {soundWaveAssetPtr=&/Athena_Emotes_DJDrop_Intro_opt}}
 {id=WavePlayer_2, properties {bLooping=true, soundWaveAssetPtr=&/Athena_Emotes_DJDrop_Loop}}
 {id=WavePlayer_3, properties {bLooping=true, soundWaveAssetPtr=&/Athena_Emotes_DJDrop_Loop_opt}}
]
properties {
 maxDistance=2097152
 soundClassObject=/Game/Sounds/FortSoundClassesAndModes/Fort_Music_Emote_1P_PSM
 volumeMultiplier=1.1
 concurrencyOverrides {resolutionRule=PreventNew}
 concurrencySet [/Game/Sounds/Concurrency_Rules/Athena_EmoteMusic_1P_Conc]
 soundSubmixSends [{minSendLevel=0, soundSubmix=&Fortnite_AudioAnalysis_Music}]
}
