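# Sound cue for the Athena Dab emote's whoosh, built from player foley samples.
# A mixer layers a random run-movement foley sound with a random swing whoosh that follows after a short fixed delay.
# Only the whoosh branch gets random pitch modulation, on top of a raised cue-wide pitch.
# The cue uses emote attenuation, the emote music sound class and concurrency set, and a send to the music audio-analysis submix.
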SoundCue /Game/Athena/Sounds/Emotes/Athena_Emote_Dab_Whoosh_Cue
root=Mixer_0
nodes [
 {id=Delay_0, children [Modulator_2], properties {delayMax=0.25, delayMin=0.25}}
 {id=Mixer_0, children [Random_1,Delay_0], properties {inputVolume [1,1]}}
 {id=Modulator_2, children [Random_0], properties {pitchMax=1.2, volumeMax=1, volumeMin=1}}
 {id=Random_0, children [WavePlayer_0,WavePlayer_1,WavePlayer_2], properties {weights [1,1,1]}}
 {id=Random_1, children [WavePlayer_5,WavePlayer_6,WavePlayer_7], properties {weights [1,1,1]}}
 {id=WavePlayer_0, properties {soundWaveAssetPtr=/Game/Sounds/Fort_Foley_Sounds/Whooshes/player_swing_whoosh_stereo_01}}
 {id=WavePlayer_1, properties {soundWaveAssetPtr=/Game/Sounds/Fort_Foley_Sounds/Whooshes/player_swing_whoosh_stereo_02}}
 {id=WavePlayer_2, properties {soundWaveAssetPtr=/Game/Sounds/Fort_Foley_Sounds/Whooshes/player_swing_whoosh_stereo_03}}
 {id=WavePlayer_5, properties {soundWaveAssetPtr=/Game/Sounds/Fort_Foley_Sounds/Moves/player_run_foley_06}}
 {id=WavePlayer_6, properties {soundWaveAssetPtr=/Game/Sounds/Fort_Foley_Sounds/Moves/player_run_foley_04}}
 {id=WavePlayer_7, properties {soundWaveAssetPtr=/Game/Sounds/Fort_Foley_Sounds/Moves/player_run_foley_05}}
]
properties {
 attenuationSettings=&Fort_Emotes_Athena_Att
 maxDistance=1600
 pitchMultiplier=1.2
 soundClassObject=&Fort_Music_Emote
 volumeMultiplier=0.6
 concurrencyOverrides {resolutionRule=PreventNew}
 concurrencySet [/Game/Sounds/Concurrency_Rules/Athena_EmoteMusic_Conc]
 soundSubmixSends [{minSendLevel=0, soundSubmix=&Fortnite_AudioAnalysis_Music}]
}
