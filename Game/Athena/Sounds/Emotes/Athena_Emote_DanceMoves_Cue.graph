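# Sound cue for the Dance Moves emote music in Athena, using the first-person emote music sound class and Athena emote attenuation.
# A quality-level switch makes the first two levels pick evenly between the soldier and ninja dance tracks.
# The last two levels always play the ninja track.
# An emote-music concurrency set governs simultaneous playback, and a submix send routes to the music audio-analysis submix.
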SoundCue /Game/Athena/Sounds/Emotes/Athena_Emote_DanceMoves_Cue
root=QualityLevel_0
nodes [
 {id=QualityLevel_0, children [Random_1,Random_1,WavePlayer_4,WavePlayer_4]}
 {id=Random_1, children [WavePlayer_3,WavePlayer_1], properties {weights [1,1]}}
 {id=WavePlayer_1, properties {soundWaveAssetPtr=/Game/Sounds/Fort_Music/Dance/ninja_dance_01}}
 {id=WavePlayer_3, properties {soundWaveAssetPtr=/Game/Sounds/Fort_Music/Dance/dance_soldier_03}}
 {id=WavePlayer_4, properties {soundWaveAssetPtr=&ninja_dance_01}}
]
properties {
 attenuationSettings=&Fort_Emotes_Athena_Att
 maxDistance=1600
 soundClassObject=/Game/Sounds/FortSoundClassesAndModes/Fort_Music_Emote_1P_PSM
 concurrencyOverrides {maxCount=1, resolutionRule=StopOldest}
 concurrencySet [/Game/Sounds/Concurrency_Rules/Athena_EmoteMusic_1P_Conc]
 soundSubmixSends [{minSendLevel=0, soundSubmix=&Fortnite_AudioAnalysis_Music}]
}
