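# Sound cue for the Flamenco emote music, using the first-person emote music sound class.
# A mixer layers a one-shot intro and a looping body, each behind its own fixed delay so the loop starts after the intro begins.
# It uses the shared first-person emote music concurrency rule and has a send to the music audio-analysis submix.
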SoundCue /Game/Athena/Sounds/Emotes/Athena_Emote_Flamenco_Music_Cue
root=Mixer_1
nodes [
 {id=Delay_0, children [WavePlayer_2], properties {delayMax=0.24, delayMin=0.24}}
 {id=Delay_2, children [WavePlayer_1], properties {delayMax=1.1, delayMin=1.1}}
 {id=Mixer_1, children [Delay_0,Delay_2], properties {inputVolume [1,1]}}
 {id=WavePlayer_1, properties {bLooping=true, soundWaveAssetPtr=&/Flamenco_2018_Emote_01_Loop}}
 {id=WavePlayer_2, properties {soundWaveAssetPtr=&/Flamenco_2018_Intro_01}}
]
properties {
 maxDistance=2097152
 soundClassObject=/Game/Sounds/FortSoundClassesAndModes/Fort_Music_Emote_1P_PSM
 volumeMultiplier=1
 concurrencyOverrides {resolutionRule=PreventNew}
 concurrencySet [/Game/Sounds/Concurrency_Rules/Athena_EmoteMusic_1P_Conc]
 soundSubmixSends [{minSendLevel=0, soundSubmix=&Fortnite_AudioAnalysis_Music}]
}
