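# Sound cue for the foley of the Flex emote in Athena.
# A random node picks one of two equally weighted foley waves and feeds a modulator at the root, with raised cue volume and pitch.
# Despite being foley, it uses the emote music sound class and concurrency set, plus a submix send to music audio analysis.
# Playback is spatialized through the shared Athena emote attenuation settings.
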SoundCue /Game/Athena/Sounds/Emotes/Athena_Emote_Flex_Foley_01_Cue
root=Modulator_0
nodes [
 {id=Modulator_0, children [Random_2]}
 {id=Random_2, children [WavePlayer_3,WavePlayer_4], properties {weights [1,1]}}
 {id=WavePlayer_3, properties {soundWaveAssetPtr=&/athena_emote_flex_foley_02}}
 {id=WavePlayer_4, properties {soundWaveAssetPtr=&/athena_emote_flex_foley_01}}
]
properties {
 attenuationSettings=&Fort_Emotes_Athena_Att
 maxDistance=1600
 pitchMultiplier=1.2
 soundClassObject=&Fort_Music_Emote
 volumeMultiplier=2
 concurrencyOverrides {resolutionRule=PreventNew}
 concurrencySet [/Game/Sounds/Concurrency_Rules/Athena_EmoteMusic_Conc]
 soundSubmixSends [{minSendLevel=0, soundSubmix=&Fortnite_AudioAnalysis_Music}]
}
