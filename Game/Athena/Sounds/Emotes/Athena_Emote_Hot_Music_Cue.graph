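# Sound Cue that plays the looping music track for the Athena "Hot" emote, using a first-person emote music sound class.
# A fixed half-second Delay node feeds from a single looping Wave Player.
# An emote-music concurrency set governs simultaneous playback.
# A submix send routes the audio to the music audio-analysis submix.
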SoundCue /Game/Athena/Sounds/Emotes/Athena_Emote_Hot_Music_Cue
root=Delay_0
nodes [
 {id=Delay_0, children [WavePlayer_0], properties {delayMax=0.5, delayMin=0.5}}
 {id=WavePlayer_0, properties {bLooping=true, soundWaveAssetPtr=&/athena_emote_hot_music}}
]
properties {
 maxDistance=2097152
 soundClassObject=/Game/Sounds/FortSoundClassesAndModes/Fort_Music_Emote_1P_PSM
 concurrencyOverrides {maxCount=1}
 concurrencySet [/Game/Sounds/Concurrency_Rules/Athena_EmoteMusic_1P_Conc]
 soundSubmixSends [{minSendLevel=0, soundSubmix=&Fortnite_AudioAnalysis_Music}]
}
