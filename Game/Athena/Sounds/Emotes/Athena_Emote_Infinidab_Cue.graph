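# Sound cue for the Infinidab emote music in Athena, using the first-person emote music sound class.
# After a short start delay, a mixer plays a one-shot intro wave alongside a further-delayed looping main wave, so the loop follows the intro's start by a fixed offset.
# It references the first-person emote music concurrency rule and sends to the music audio-analysis submix.
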SoundCue /Game/Athena/Sounds/Emotes/Athena_Emote_Infinidab_Cue
root=Delay_1
nodes [
 {id=Delay_0, children [WavePlayer_1], properties {delayMax=0.44, delayMin=0.44}}
 {id=Delay_1, children [Mixer_0], properties {delayMax=0.25, delayMin=0.25}}
 {id=Mixer_0, children [WavePlayer_0,Delay_0], properties {inputVolume [1,1]}}
 {id=WavePlayer_0, properties {soundWaveAssetPtr=&/Athena_Emote_Infinidab_Intro}}
 {id=WavePlayer_1, properties {bLooping=true, soundWaveAssetPtr=&/Athena_Emote_Infinidab}}
]
properties {
 maxDistance=2097152
 soundClassObject=/Game/Sounds/FortSoundClassesAndModes/Fort_Music_Emote_1P_PSM
 volumeMultiplier=1
 concurrencyOverrides {resolutionRule=PreventNew}
 concurrencySet [/Game/Sounds/Concurrency_Rules/Athena_EmoteMusic_1P_Conc]
 soundSubmixSends [{minSendLevel=0, soundSubmix=&Fortnite_AudioAnalysis_Music}]
}
