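# Sound cue for the Athena Snap emote, playing its audio as spatialized emote music.
# A Random node picks one of two Snap waves with equal weight and plays it at reduced volume.
# It uses the shared emote attenuation, Music Emote sound class and emote-music concurrency set, and has a send to the music audio-analysis submix.
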SoundCue /Game/Athena/Sounds/Emotes/Athena_Emote_Snap_Cue
root=Random_0
nodes [
 {id=Random_0, children [WavePlayer_1,WavePlayer_2], properties {weights [1,1]}}
 {id=WavePlayer_1, properties {soundWaveAssetPtr=&/Emote_Snap_01}}
 {id=WavePlayer_2, properties {soundWaveAssetPtr=&/Emote_Snap_02}}
]
properties {
 attenuationSettings=&Fort_Emotes_Athena_Att
 maxDistance=1600
 soundClassObject=&Fort_Music_Emote
 volumeMultiplier=0.6
 concurrencyOverrides {resolutionRule=PreventNew}
 concurrencySet [/Game/Sounds/Concurrency_Rules/Athena_EmoteMusic_Conc]
 soundSubmixSends [{minSendLevel=0, soundSubmix=&Fortnite_AudioAnalysis_Music}]
}
