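# Sound cue for the Yeet emote in Athena, playing foley sounds with emote attenuation.
# A mixer layers the Yeet foley, delayed briefly, at full level with a quieter, immediate foley borrowed from the KungFuSalute emote.
# Although both waves are foley, the cue uses the emote music sound class and concurrency set, and sends to a music audio-analysis submix.
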SoundCue /Game/Athena/Sounds/Emotes/Athena_Emote_Yeet_Cue
root=Mixer_0
nodes [
 {id=Delay_0, children [WavePlayer_5], properties {delayMax=0.4, delayMin=0.4}}
 {id=Mixer_0, children [Delay_0,WavePlayer_3], properties {inputVolume [1,0.3]}}
 {id=WavePlayer_3, properties {soundWaveAssetPtr=&/Athena_Emote_KungFuSalute_Foley03}}
 {id=WavePlayer_5, properties {soundWaveAssetPtr=&/Emote_Yeet_Foley_01}}
]
properties {
 attenuationSettings=&Fort_Emotes_Athena_Att
 maxDistance=1600
 soundClassObject=&Fort_Music_Emote
 volumeMultiplier=0.6
 concurrencyOverrides {resolutionRule=PreventNew}
 concurrencySet [/Game/Sounds/Concurrency_Rules/Athena_EmoteMusic_Conc]
 soundSubmixSends [{minSendLevel=0, soundSubmix=&Fortnite_AudioAnalysis_Music}]
}
