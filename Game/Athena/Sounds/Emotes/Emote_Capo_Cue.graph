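# Sound cue that plays the music for the Capo emote in Athena, using a first-person emote music sound class.
# A mixer starts the intro wave at once alongside a looping wave held back by a fixed delay of about 1.6 seconds, so the loop follows the intro.
# A concurrency set governs simultaneous instances, and a submix send feeds Fortnite's music audio analysis.
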SoundCue /Game/Athena/Sounds/Emotes/Emote_Capo_Cue
root=Mixer_0
nodes [
 {id=Delay_0, children [WavePlayer_2], properties {delayMax=1.647, delayMin=1.647}}
 {id=Mixer_0, children [WavePlayer_1,Delay_0], properties {inputVolume [1,1]}}
 {id=WavePlayer_1, properties {soundWaveAssetPtr=&/Emote_Capo_Intro}}
 {id=WavePlayer_2, properties {bLooping=true, soundWaveAssetPtr=&/Emote_Capo_Loop}}
]
properties {
 maxDistance=2097152
 soundClassObject=/Game/Sounds/FortSoundClassesAndModes/Fort_Music_Emote_1P_PSM
 volumeMultiplier=1
 concurrencyOverrides {maxCount=1, resolutionRule=PreventNew}
 concurrencySet [/Game/Sounds/Concurrency_Rules/Athena_EmoteMusic_1P_Conc]
 soundSubmixSends [{minSendLevel=0, soundSubmix=&Fortnite_AudioAnalysis_Music}]
}
