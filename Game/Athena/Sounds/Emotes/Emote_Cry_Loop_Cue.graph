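# Sound cue for the crying emote's audio in Athena, playing an intro that leads into a loop.
# A mixer starts the one-shot cry intro immediately and, through a fixed delay of about a second, layers in the looping cry so it follows the intro.
# It uses emote attenuation, the emote music sound class and concurrency rules, and sends to a music audio-analysis submix.
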SoundCue /Game/Athena/Sounds/Emotes/Emote_Cry_Loop_Cue
root=Mixer_0
nodes [
 {id=Delay_0, children [WavePlayer_0], properties {delayMax=1.116, delayMin=1.116}}
 {id=Mixer_0, children [WavePlayer_1,Delay_0], properties {inputVolume [1,1]}}
 {id=WavePlayer_0, properties {bLooping=true, soundWaveAssetPtr=&/Emote_Cry_Loop}}
 {id=WavePlayer_1, properties {soundWaveAssetPtr=&/Emote_Cry_Start}}
]
properties {
 attenuationSettings=&Fort_Emotes_Athena_Att
 maxDistance=1600
 soundClassObject=&Fort_Music_Emote
 concurrencySet [/Game/Sounds/Concurrency_Rules/Athena_EmoteMusic_Conc]
 soundSubmixSends [{minSendLevel=0, soundSubmix=&Fortnite_AudioAnalysis_Music}]
}
