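# Sound cue for the catch foley in the Juggler emote, using the juggle foley recordings.
# A quality switch sends its first two levels to an equally weighted random pick of nine waves, pitch unmodulated.
# Its last two levels play only take 03 with randomized pitch and volume.
# It uses emote attenuation but is routed as emote music, with that sound class, concurrency set and a music analysis submix send.
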SoundCue /Game/Athena/Sounds/Emotes/Emote_Juggler_Catch_Foley_Cue
root=QualityLevel_0
nodes [
 {id=Modulator_0, children [WavePlayer_0], properties {pitchMax=1.15, pitchMin=0.85, volumeMin=0.6}}
 {id=Modulator_1, children [Random_0], properties {pitchMax=1, pitchMin=1}}
 {id=QualityLevel_0, children [Modulator_1,Modulator_1,Modulator_0,Modulator_0]}
 {
  id=Random_0
  children [WavePlayer_1, WavePlayer_2, WavePlayer_3, WavePlayer_4, WavePlayer_5, WavePlayer_6, WavePlayer_8, WavePlayer_9, WavePlayer_10]
  properties {weights [1,1,1,1,1,1,1,1,1]}
 }
 {id=WavePlayer_0, properties {soundWaveAssetPtr=&/Emote_Juggler_Juggle_Foley_03}}
 {id=WavePlayer_1, properties {soundWaveAssetPtr=&/Emote_Juggler_Juggle_Foley_01}}
 {id=WavePlayer_10, properties {soundWaveAssetPtr=&/Emote_Juggler_Juggle_Foley_09}}
 {id=WavePlayer_2, properties {soundWaveAssetPtr=&/Emote_Juggler_Juggle_Foley_02}}
 {id=WavePlayer_3, properties {soundWaveAssetPtr=&Emote_Juggler_Juggle_Foley_03}}
 {id=WavePlayer_4, properties {soundWaveAssetPtr=&/Emote_Juggler_Juggle_Foley_04}}
 {id=WavePlayer_5, properties {soundWaveAssetPtr=&/Emote_Juggler_Juggle_Foley_05}}
 {id=WavePlayer_6, properties {soundWaveAssetPtr=&/Emote_Juggler_Juggle_Foley_06}}
 {id=WavePlayer_8, properties {soundWaveAssetPtr=&/Emote_Juggler_Juggle_Foley_07}}
 {id=WavePlayer_9, properties {soundWaveAssetPtr=&/Emote_Juggler_Juggle_Foley_08}}
]
properties {
 attenuationSettings=&Fort_Emotes_Athena_Att
 maxDistance=1600
 soundClassObject=&Fort_Music_Emote
 volumeMultiplier=1.5
 concurrencyOverrides {resolutionRule=StopOldest}
 concurrencySet [/Game/Sounds/Concurrency_Rules/Athena_EmoteMusic_Conc]
 soundSubmixSends [{minSendLevel=0, soundSubmix=&Fortnite_AudioAnalysis_Music}]
}
