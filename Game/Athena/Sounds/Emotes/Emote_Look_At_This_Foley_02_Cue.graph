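# Sound cue for the second foley layer of the Athena "Look At This" emote, built from player running foley waves.
# A quality-level switch sends its first two levels to a modulated random pick of four waves, and its last two to one wave with wider pitch and volume modulation.
# Despite being foley, it uses emote music class and concurrency, emote attenuation, and a music-analysis submix send.
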SoundCue /Game/Athena/Sounds/Emotes/Emote_Look_At_This_Foley_02_Cue
root=QualityLevel_0
nodes [
 {id=Mixer_0, children [WavePlayer_0], properties {inputVolume [0.75]}}
 {id=Modulator_0, children [Mixer_0], properties {pitchMax=1.15, pitchMin=0.75, volumeMin=0.3}}
 {id=Modulator_1, children [Random_2]}
 {id=QualityLevel_0, children [Modulator_1,Modulator_1,Modulator_0,Modulator_0]}
 {id=Random_2, children [WavePlayer_1,WavePlayer_3,WavePlayer_7,WavePlayer_2], properties {weights [1,1,1,1]}}
 {id=WavePlayer_0, properties {soundWaveAssetPtr=/Game/Sounds/Fort_Foley_Sounds/Moves/player_run_foley_01}}
 {id=WavePlayer_1, properties {soundWaveAssetPtr=/Game/Sounds/Fort_Foley_Sounds/Moves/player_run_foley_04}}
 {id=WavePlayer_2, properties {soundWaveAssetPtr=&player_run_foley_01}}
 {id=WavePlayer_3, properties {soundWaveAssetPtr=/Game/Sounds/Fort_Foley_Sounds/Moves/player_run_foley_09}}
 {id=WavePlayer_7, properties {soundWaveAssetPtr=/Game/Sounds/Fort_Foley_Sounds/Moves/player_run_foley_05}}
]
properties {
 attenuationSettings=&Fort_Emotes_Athena_Att
 maxDistance=1600
 pitchMultiplier=1.2
 soundClassObject=&Fort_Music_Emote
 volumeMultiplier=0.3
 concurrencyOverrides {resolutionRule=PreventNew}
 concurrencySet [/Game/Sounds/Concurrency_Rules/Athena_EmoteMusic_Conc]
 soundSubmixSends [{minSendLevel=0, soundSubmix=&Fortnite_AudioAnalysis_Music}]
}
