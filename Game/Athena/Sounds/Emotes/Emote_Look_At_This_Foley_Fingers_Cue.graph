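# Sound cue for the finger foley of the Look At This emote, with emote attenuation, music sound class and concurrency.
# A quality-level root sends its first two slots to a pitch- and volume-modulated random pick of six waves.
# The last two slots use a cheaper random pick of two of those waves through a reduced-volume mixer.
# The cue lowers overall pitch and volume and sends to a music audio-analysis submix.
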SoundCue /Game/Athena/Sounds/Emotes/Emote_Look_At_This_Foley_Fingers_Cue
root=QualityLevel_0
nodes [
 {id=Mixer_0, children [Random_1], properties {inputVolume [0.75]}}
 {id=Modulator_1, children [Random_0], properties {pitchMax=1.1, pitchMin=0.8, volumeMin=0.6}}
 {id=QualityLevel_0, children [Modulator_1,Modulator_1,Mixer_0,Mixer_0]}
 {id=Random_0, children [WavePlayer_1,WavePlayer_2,WavePlayer_3,WavePlayer_4,WavePlayer_5,WavePlayer_6], properties {weights [1,1,1,1,1,1]}}
 {id=Random_1, children [WavePlayer_7,WavePlayer_0], properties {weights [1,1]}}
 {id=WavePlayer_0, properties {soundWaveAssetPtr=&/Emote_Look_At_This_Finger_Foley_05}}
 {id=WavePlayer_1, properties {soundWaveAssetPtr=&/Emote_Look_At_This_Finger_Foley_01}}
 {id=WavePlayer_2, properties {soundWaveAssetPtr=&/Emote_Look_At_This_Finger_Foley_02}}
 {id=WavePlayer_3, properties {soundWaveAssetPtr=&/Emote_Look_At_This_Finger_Foley_03}}
 {id=WavePlayer_4, properties {soundWaveAssetPtr=&/Emote_Look_At_This_Finger_Foley_04}}
 {id=WavePlayer_5, properties {soundWaveAssetPtr=&Emote_Look_At_This_Finger_Foley_05}}
 {id=WavePlayer_6, properties {soundWaveAssetPtr=&/Emote_Look_At_This_Finger_Foley_06}}
 {id=WavePlayer_7, properties {soundWaveAssetPtr=&Emote_Look_At_This_Finger_Foley_02}}
]
properties {
 attenuationSettings=&Fort_Emotes_Athena_Att
 maxDistance=1600
 pitchMultiplier=0.7
 soundClassObject=&Fort_Music_Emote
 volumeMultiplier=0.5
 concurrencyOverrides {resolutionRule=StopOldest}
 concurrencySet [/Game/Sounds/Concurrency_Rules/Athena_EmoteMusic_Conc]
 soundSubmixSends [{minSendLevel=0, soundSubmix=&Fortnite_AudioAnalysis_Music}]
}
