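# Sound cue for the music of the Tai Chi emote in Athena.
# A mixer plays a one-shot intro wave alongside a looping wave that starts after a short fixed delay, so the intro leads into the loop.
# It uses emote attenuation, a first-person emote music sound class and concurrency rule, and has a send to a music audio-analysis submix.
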SoundCue /Game/Athena/Sounds/Emotes/Emote_TaiChi_Cue
root=Mixer_0
nodes [
 {id=Delay_0, children [WavePlayer_3], properties {delayMax=0.239, delayMin=0.239}}
 {id=Mixer_0, children [WavePlayer_2,Delay_0], properties {inputVolume [1,1]}}
 {id=WavePlayer_2, properties {soundWaveAssetPtr=&/Emote_TaiChi_Intro_01}}
 {id=WavePlayer_3, properties {bLooping=true, soundWaveAssetPtr=&/Emote_TaiChi_Loop_01}}
]
properties {
 attenuationSettings=&Fort_Emotes_Athena_Att
 maxDistance=1600
 soundClassObject=/Game/Sounds/FortSoundClassesAndModes/Fort_Music_Emote_1P_PSM
 volumeMultiplier=1.5
 concurrencySet [/Game/Sounds/Concurrency_Rules/Athena_EmoteMusic_1P_Conc]
 soundSubmixSends [{minSendLevel=0, soundSubmix=&Fortnite_AudioAnalysis_Music}]
}
