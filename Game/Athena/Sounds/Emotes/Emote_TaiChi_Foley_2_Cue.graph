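# Sound cue for the second foley layer of the Tai Chi emote in Athena, built from borrowed player running foley.
# A random node picks one of four equally weighted run foley waves, which passes through two chained modulators at default settings.
# The cue plays quiet and pitched up, with emote attenuation, the emote music sound class and concurrency, and a send to the music analysis submix.
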SoundCue /Game/Athena/Sounds/Emotes/Emote_TaiChi_Foley_2_Cue
root=Modulator_0
nodes [
 {id=Modulator_0, children [Modulator_2]}
 {id=Modulator_2, children [Random_0]}
 {id=Random_0, children [WavePlayer_2,WavePlayer_4,WavePlayer_1,WavePlayer_3], properties {weights [1,1,1,1]}}
 {id=WavePlayer_1, properties {soundWaveAssetPtr=/Game/Sounds/Fort_Foley_Sounds/Moves/player_run_foley_05}}
 {id=WavePlayer_2, properties {soundWaveAssetPtr=/Game/Sounds/Fort_Foley_Sounds/Moves/player_run_foley_04}}
 {id=WavePlayer_3, properties {soundWaveAssetPtr=/Game/Sounds/Fort_Foley_Sounds/Moves/player_run_foley_01}}
 {id=WavePlayer_4, properties {soundWaveAssetPtr=/Game/Sounds/Fort_Foley_Sounds/Moves/player_run_foley_09}}
]
properties {
 attenuationSettings=&Fort_Emotes_Athena_Att
 maxDistance=1600
 pitchMultiplier=1.2
 soundClassObject=&Fort_Music_Emote
 volumeMultiplier=0.3
 concurrencyOverrides {resolutionRule=PreventNew}
 concurrencySet [/Game/Sounds/Concurrency_Rules/Athena_EmoteMusic_Conc]
 soundSubmixSends [{minSendLevel=0, soundSubmix=&Fortnite_AudioAnalysis_Music}]
}
