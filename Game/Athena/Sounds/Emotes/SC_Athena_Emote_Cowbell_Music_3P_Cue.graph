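# Third-person music cue for the Athena Cowbell emote, playing an intro followed by a looping bed.
# A short initial delay feeds a mixer that starts the intro wave immediately and the looping wave after a fixed delay, stitching the loop onto the intro by timing rather than concatenation.
# It uses emote attenuation, music concurrency and play-when-silent virtualization, and sends to a music audio-analysis submix.
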SoundCue /Game/Athena/Sounds/Emotes/SC_Athena_Emote_Cowbell_Music_3P_Cue
root=Delay_1
nodes [
 {id=Delay_0, children [WavePlayer_3], properties {delayMax=3.69, delayMin=3.69}}
 {id=Delay_1, children [Mixer_0], properties {delayMax=0.2306, delayMin=0.2306}}
 {id=Mixer_0, children [WavePlayer_0,Delay_0], properties {inputVolume [1,1]}}
 {id=WavePlayer_0, properties {soundWaveAssetPtr=&/Athena_Emotes_Music_Cowbell_Intro}}
 {id=WavePlayer_3, properties {bLooping=true, soundWaveAssetPtr=&/Athena_Emotes_Music_Cowbell_Loop}}
]
properties {
 attenuationSettings=&Fort_Emotes_Athena_Att
 maxDistance=1600
 soundClassObject=/Game/Sounds/FortSoundClassesAndModes/Fort_Music_Emote_3P
 virtualizationMode=PlayWhenSilent
 volumeMultiplier=1
 concurrencyOverrides {maxCount=2, volumeScale=0.5}
 concurrencySet [/Game/Sounds/Concurrency_Rules/Athena_EmoteMusic_Conc]
 soundSubmixSends [{minSendLevel=0, soundSubmix=&Fortnite_AudioAnalysis_Music}]
}
