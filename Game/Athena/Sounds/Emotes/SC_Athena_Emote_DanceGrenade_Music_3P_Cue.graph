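# Sound cue for the third-person, spatialized music loop of the Athena Dance Grenade emote.
# A looping wave player feeds an Enveloper whose volume curve fades the track in from silence over the first second.
# It uses dedicated attenuation, the third-person emote music sound class and shared emote-music concurrency, and keeps playing when silent.
# A submix send also routes it to the music audio-analysis submix.
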SoundCue /Game/Athena/Sounds/Emotes/SC_Athena_Emote_DanceGrenade_Music_3P_Cue
root=Enveloper_2
nodes [
 {
  id=Enveloper_2
  children [WavePlayer_0]
  properties {pitchCurve {editorCurveData {keys [{time=0}]}}, volumeCurve {editorCurveData {keys [<at 0 0>,<at 1 1>]}}}
 }
 {id=WavePlayer_0, properties {bLooping=true, soundWaveAssetPtr=/Game/Athena/Sounds/DanceGrenade/music_dancegrenade_loop}}
]
properties {
 attenuationSettings=/Game/Sounds/Attenuation/Fort_Emotes_DanceGrenadeMusic_Att
 maxDistance=2350
 soundClassObject=/Game/Sounds/FortSoundClassesAndModes/Fort_Music_Emote_3P
 virtualizationMode=PlayWhenSilent
 concurrencyOverrides {maxCount=1, volumeScale=0.5}
 concurrencySet [/Game/Sounds/Concurrency_Rules/Athena_EmoteMusic_Conc]
 soundSubmixSends [{minSendLevel=0, soundSubmix=&Fortnite_AudioAnalysis_Music}]
}
