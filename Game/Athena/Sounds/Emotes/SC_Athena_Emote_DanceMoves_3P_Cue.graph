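# Sound cue for third-person Dance Moves emote music in Athena.
# A quality-level node sends its first two slots to an equal-weight random pick between a soldier and a ninja dance track, and its last two to the ninja track alone.
# Playback is attenuated, limited by emote music concurrency, kept playing when silent, and sent to a music audio-analysis submix.
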SoundCue /Game/Athena/Sounds/Emotes/SC_Athena_Emote_DanceMoves_3P_Cue
root=QualityLevel_0
nodes [
 {id=QualityLevel_0, children [Random_0,Random_0,WavePlayer_1,WavePlayer_1]}
 {id=Random_0, children [WavePlayer_2,WavePlayer_0], properties {weights [1,1]}}
 {id=WavePlayer_0, properties {soundWaveAssetPtr=/Game/Sounds/Fort_Music/Dance/ninja_dance_01}}
 {id=WavePlayer_1, properties {soundWaveAssetPtr=&ninja_dance_01}}
 {id=WavePlayer_2, properties {soundWaveAssetPtr=/Game/Sounds/Fort_Music/Dance/dance_soldier_03}}
]
properties {
 attenuationSettings=&Fort_Emotes_Athena_Att
 maxDistance=1600
 soundClassObject=/Game/Sounds/FortSoundClassesAndModes/Fort_Music_Emote_3P
 virtualizationMode=PlayWhenSilent
 concurrencyOverrides {maxCount=2, volumeScale=0.5}
 concurrencySet [/Game/Sounds/Concurrency_Rules/Athena_EmoteMusic_Conc]
 soundSubmixSends [{minSendLevel=0, soundSubmix=&Fortnite_AudioAnalysis_Music}]
}
