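# Sound cue for the Flamenco emote's music as heard in third person, routed to the emote music sound class with emote attenuation and concurrency.
# A mixer plays a one-shot intro after a short delay alongside a looping track delayed about a second, so the loop follows the intro.
# It keeps playing virtualized when silent and sends to a music audio-analysis submix.
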SoundCue /Game/Athena/Sounds/Emotes/SC_Athena_Emote_Flamenco_Music_3P_Cue
root=Mixer_2
nodes [
 {id=Delay_4, children [WavePlayer_4], properties {delayMax=1.1, delayMin=1.1}}
 {id=Delay_5, children [WavePlayer_5], properties {delayMax=0.24, delayMin=0.24}}
 {id=Mixer_2, children [Delay_5,Delay_4], properties {inputVolume [1,1]}}
 {id=WavePlayer_4, properties {bLooping=true, soundWaveAssetPtr=&/Flamenco_2018_Emote_01_Loop}}
 {id=WavePlayer_5, properties {soundWaveAssetPtr=&/Flamenco_2018_Intro_01}}
]
properties {
 attenuationSettings=&Fort_Emotes_Athena_Att
 maxDistance=1600
 soundClassObject=/Game/Sounds/FortSoundClassesAndModes/Fort_Music_Emote_3P
 virtualizationMode=PlayWhenSilent
 volumeMultiplier=1
 concurrencyOverrides {resolutionRule=PreventNew}
 concurrencySet [/Game/Sounds/Concurrency_Rules/Athena_EmoteMusic_Conc]
 soundSubmixSends [{minSendLevel=0, soundSubmix=&Fortnite_AudioAnalysis_Music}]
}
