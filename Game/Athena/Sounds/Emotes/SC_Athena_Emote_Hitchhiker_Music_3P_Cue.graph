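# Sound cue for the Hitchhiker emote music in Athena, the third-person version.
# A looping wave player feeds a modulator that lowers volume to a fixed level with no pitch variation.
# Emote attenuation, the third-person emote music sound class and shared emote music concurrency govern playback, with a prevent-new override present.
# It keeps playing when silent and sends to the music audio analysis submix.
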SoundCue /Game/Athena/Sounds/Emotes/SC_Athena_Emote_Hitchhiker_Music_3P_Cue
root=Modulator_0
nodes [
 {id=Modulator_0, children [WavePlayer_1], properties {pitchMax=1, pitchMin=1, volumeMax=0.6, volumeMin=0.6}}
 {id=WavePlayer_1, properties {bLooping=true, soundWaveAssetPtr=&/Athena_Emote_Hitchhiker_Music}}
]
properties {
 attenuationSettings=&Fort_Emotes_Athena_Att
 maxDistance=1600
 soundClassObject=/Game/Sounds/FortSoundClassesAndModes/Fort_Music_Emote_3P
 virtualizationMode=PlayWhenSilent
 volumeMultiplier=1
 concurrencyOverrides {resolutionRule=PreventNew}
 concurrencySet [/Game/Sounds/Concurrency_Rules/Athena_EmoteMusic_Conc]
 soundSubmixSends [{minSendLevel=0, soundSubmix=&Fortnite_AudioAnalysis_Music}]
}
