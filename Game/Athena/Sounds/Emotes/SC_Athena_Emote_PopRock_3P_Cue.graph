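# Third-person sound cue for the PopRock emote music in Athena, spatialized with emote attenuation and the 3P emote music sound class.
# A root mixer layers the looping music wave, run through an Enveloper for a short volume ramp-in, with a quieter one-shot crash.
# It keeps playing virtualized when silent, joins the emote music concurrency group and has a send to the music audio-analysis submix.
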SoundCue /Game/Athena/Sounds/Emotes/SC_Athena_Emote_PopRock_3P_Cue
root=Mixer_0
nodes [
 {
  id=Enveloper_1
  children [WavePlayer_2]
  properties {pitchCurve {editorCurveData {keys [{time=0}]}}, volumeCurve {editorCurveData {keys [<at 0 0.7>,<at 0.3 1>]}}}
 }
 {id=Mixer_0, children [Enveloper_1,WavePlayer_3], properties {inputVolume [1,0.3]}}
 {id=WavePlayer_2, properties {bLooping=true, soundWaveAssetPtr=&/Emote_PopRock_01}}
 {id=WavePlayer_3, properties {soundWaveAssetPtr=&/Emote_PopRock_crash_01}}
]
properties {
 attenuationSettings=&Fort_Emotes_Athena_Att
 maxDistance=1600
 soundClassObject=/Game/Sounds/FortSoundClassesAndModes/Fort_Music_Emote_3P
 virtualizationMode=PlayWhenSilent
 volumeMultiplier=0.7
 concurrencyOverrides {resolutionRule=PreventNew}
 concurrencySet [/Game/Sounds/Concurrency_Rules/Athena_EmoteMusic_Conc]
 soundSubmixSends [{minSendLevel=0, soundSubmix=&Fortnite_AudioAnalysis_Music}]
}
